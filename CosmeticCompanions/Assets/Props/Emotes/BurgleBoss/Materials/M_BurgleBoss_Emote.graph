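# Skeletal mesh material for the BurgleBoss emote, adding a MidasTech gold transformation to the 2023 character parent's PreFX/PostFX functions.
# A pre-skinned sphere mask, with gold_radius setting size and a stage-driven curve moving its center, is eroded by world-aligned noise.
# Hard or soft body masks and emissive edge masks support blending into metallic gold with a glowing transition.
# Pre-skinned normal Z tints upward-facing gold lighter.
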Material /CosmeticCompanions/Assets/Props/Emotes/BurgleBoss/Materials/M_BurgleBoss_Emote
graph {
 properties {
  bAutomaticallySetUsageInEditor=false
  bUseMaterialAttributes=true
  bUsedWithMeshDeformer=true
  bUsedWithMorphTargets=true
  bUsedWithSkeletalMesh=true
  pixelDepthOffsetMode=PDOM_Legacy
  shadingModel=MSM_FromMaterialExpression
  usedShadingModels=Default Lit | Subsurface | Clear Coat | Subsurface Profile | Cloth
 }
 nodes [
  n1 = ScalarParameter() {defaultValue=100, group=MidasTech_GoldMask, parameterName=GM_CurrentStage, sortPriority=3}
  n2 = ScalarParameter() {group=MidasTech_GoldMask, parameterName=GM_StartAtStage, sortPriority=3}
  n3 = ScalarParameter() {defaultValue=3, group=MidasTech_GoldMask, parameterName=GM_StopAtStage, sortPriority=3}
  n4 = Clamp(None=LinearInterpolate(A=Subtract(A=@n1), B=@n1, Alpha=NamedRerouteUsage()) {desc=NativeLerpTowards}, Min=@n2, Max=@n3)
  n5 = CurveAtlasRowParameter(CurveTime=Divide(A=Subtract(A=@n4, B=@n2), B=Subtract(A=@n3, B=@n2))) {
   atlas=&AthenaCurveAtlas01
   curve=/Game/Athena/LinearColorCurves/Curve_SuperlevelMask_Summer
   defaultValue=289
   group=MidasTech_GoldMask
   parameterName=GM_GrowingCurve
  }
  n6 = ChannelMaskParameter(None=@n5) {group=MidasTech_GoldMask, parameterName=GM_GrowingCurveChannel}
  LinearInterpolate(
   A=ScalarParameter() {defaultValue=10, group=MidasTech_GoldMask, parameterName=GM_StartRadius}
   B=ScalarParameter() {defaultValue=227, group=MidasTech_GoldMask, parameterName=GM_EndRadius}
   Alpha=@n6.R
  )
  n7 = MaterialFunctionCall() {materialFunction=&MF_CharacterParent_2023_PreFX}
  MaterialFunctionCall(Material=NamedRerouteUsage(), M=@n7.M, Pre Skinned Local Position=@n7.M, Distance Blend=@n7.M, Fresnel=@n7.M) {materialFunction=&MF_CharacterParent_2023_PostFX}
  n8 = NamedRerouteUsage()
  n9 = SmoothStep(
   Min=ScalarParameter() {defaultValue=0.4, group=MidasTech_GoldMask, parameterName=GM_HardMask_Min, sortPriority=100}
   Max=ScalarParameter() {defaultValue=0.45, group=MidasTech_GoldMask, parameterName=GM_HardMask_Max, sortPriority=100}
   Value=Saturate(None=@n8)
  )
  NamedRerouteDeclaration(None=Saturate(None=StaticSwitchParameter(True=@n9, False=@n8) {defaultValue=true, group=MidasTech_GoldMask, parameterName=GM_UseHardEdgedMask})) {name=BodyMask, nodeColor=<rgba 0 1 0.6016 1>}
  n10 = SmoothStep(
   Min=ScalarParameter() {defaultValue=0.39, group=MidasTech_GoldMask, parameterName=GM_HardMask_Min_Emissive, sortPriority=101}
   Max=ScalarParameter() {defaultValue=0.47, group=MidasTech_GoldMask, parameterName=GM_HardMask_Max_Emissive, sortPriority=101}
   Value=NamedRerouteUsage()
  )
  n11 = OneMinus(None=@n10)
  NamedRerouteDeclaration(None=If(A=@n10, B=Constant() {r=0.9}, "A > B"=@n11, "A == B"=@n11, "A < B"=@n10)) {name=EdgeMask, nodeColor=<rgba 1 0.5625 0 1>}
  NamedRerouteDeclaration(None=@n7.Result) {name=IN_Material, nodeColor=<rgba 1 0.0703 0 1>}
  n12 = MaterialFunctionCall(X=NamedRerouteUsage(), Y=NamedRerouteUsage(), Z=Constant()) {materialFunction=&MakeFloat3}
  NamedRerouteDeclaration(None=Saturate(None=ComponentMask(None=@n12.Result) {r=true})) {name=PEV_Mask, nodeColor=<rgba 0 0.0422 1 1>}
  NamedRerouteDeclaration(None=Saturate(None=ComponentMask(None=@n12.Result) {g=true})) {name=PEV_Emissive, nodeColor=<rgba 0 0.1562 1 1>}
  n13 = NamedRerouteUsage()
  n14 = Saturate(None=Add(A=ComponentMask(None=@n13) {g=true}, B=ScalarParameter() {defaultValue=0.3, group=AuraCop, parameterName=GMapIntensity}))
  n15 = ComponentMask(None=VertexInterpolator(VS=PreSkinnedNormal()).PS) {b=true}
  n16 = LinearInterpolate(
   A=VectorParameter() {group=AuraCop, parameterName=GoldBaseColor, defaultValue=<rgba 0.5 0.221 0.0183 1>}.RGB
   B=VectorParameter() {group=AuraCop, parameterName=GoldBaseColor_Y_Top, defaultValue=<rgba 0.8073 0.5146 0.1598 1>}.RGB
   Alpha=SmoothStep(
    Min=ScalarParameter() {defaultValue=0.5, group=AuraCop, parameterName=GoldBaseColor_Min}
    Max=ScalarParameter() {defaultValue=1, group=AuraCop, parameterName=GoldBaseColor_Max}
    Value=Saturate(None=@n15)
   ) {constMin=0.5}
  )
  n17 = Constant() {r=1}
  n18 = Constant()
  n19 = SetMaterialAttributes!(
   MaterialAttributes=NamedRerouteUsage()
   Base Color=NamedRerouteDeclaration(None=Multiply(A=@n14, B=@n16)) {name=GoldColor, nodeColor=<rgba 1 0 0.75 1>}
   Specular=@n17
   Roughness=Constant() {r=0.25}
   Metallic=@n17
   Emissive Color=@n18
   Normal
   Opacity=@n18
   Shading Model=ShadingModel()
  ) {attributeSetTypes [
    #1
    #2
    #3
    #4
    #5
    #6
    #7
    #8
   ]}
  n20 = NamedRerouteUsage()
  n21 = Constant()
  n22 = If(
   A=ScalarParameter() {defaultValue=100, group=MidasTech_GoldMask, parameterName=GM_CurrentStage, sortPriority=3}
   "A > B"=Constant() {r=1}
   "A == B"=@n21
   "A < B"=@n21
  )
  n23 = GetMaterialAttributes(None=BlendMaterialAttributes(A=NamedRerouteUsage(), B=@n19, Alpha=Multiply(A=@n20, B=@n22))) {attributeGetTypes [#5, #4]}
  NamedRerouteDeclaration(None=@n23.Emissive Color) {name=CurrentEmissive, nodeColor=<rgba 0.9922 1 0 1>}
  NamedRerouteDeclaration(None=@n23.Metallic) {name=CurrentMetallic, nodeColor=<rgba 0 1 0.6953 1>}
  n24 = NamedRerouteUsage()
  n25 = VectorParameter() {group=MidasTech_GoldMask, parameterName=GM_Transition_EmissiveColor, sortPriority=101, defaultValue=<rgba 20 8.8102 1 1>}
  n26 = ScalarParameter() {defaultValue=1, group=MidasTech_GoldMask, parameterName=GM_Transition_EmissiveIntensity}
  n27 = LinearInterpolate(A=@n24, B=Add(A=@n24, B=Multiply(A=@n25.RGB, B=@n26)), Alpha=NamedRerouteUsage())
  n28 = OneMinus(None=NamedRerouteUsage())
  NamedRerouteDeclaration(None=SetMaterialAttributes!(MaterialAttributes=@n23.MaterialAttributes, Emissive Color=Multiply(A=@n27, B=@n28)) {attributeSetTypes [#5]}) {name=GoldEffectMaterial, nodeColor=<rgba 0 0.5078 1 1>}
  n29 = Subtract(A=Time(), B=ScalarParameter() {defaultValue=-99999, group=MidasTech_GoldMask, parameterName=GM_GameTimeFromBP, sortPriority=100})
  n30 = Add(A=ScalarParameter() {defaultValue=2.5, group=MidasTech_GoldMask, parameterName=GM_Animation_Duration, sortPriority=101}) {constB=0}
  NamedRerouteDeclaration(None=Saturate(None=Divide(A=@n29, B=@n30))) {name=TimeNative, nodeColor=<rgba 0.2891 1 0 1>}
  n31 = TextureObjectParameter() {
   group=MidasTech_GoldMask
   parameterName=GM_TransitionNoise
   texture=/Game/Characters/Player/Male/Medium/Bodies/M_MED_Viper/Textures/Glitch/T_SoftNoise
  }
  n32 = LocalPosition() {localOrigin=InstancePreSkinning}
  n33 = VertexInterpolator(VS=@n32.XYZ)
  n34 = VectorParameter() {group=MidasTech_GoldMask, parameterName=GM_Backpack_PreskinnedPositionOffset, defaultValue=<rgba 0 0 0 1>}
  n35 = NamedRerouteDeclaration(None=Add(A=ShaderStageSwitch(PixelShader=@n33.PS, VertexShader=@n32.XYZ), B=@n34.RGB)) {name=PreSkinnedPositionOffset, nodeColor=<rgba 1 0.8906 0 1>}
  n36 = VectorParameter() {group=MidasTech_GoldMask, parameterName=GM_TransitionNoise_Offset, defaultValue=<rgba 0 0 0 1>}
  n37 = PreSkinnedNormal()
  n38 = VertexInterpolator(VS=@n37)
  n39 = ChannelMaskParameter(None=MaterialFunctionCall(
    TextureObject=@n31
    TextureSize=VectorParameter() {group=MidasTech_GoldMask, parameterName=GM_TransitionNoise_Tiling, defaultValue=<rgba 40 40 40 1>}.RGB
    WorldPosition=Add(A=@n35, B=@n36.RGB)
    World Space Normal=ShaderStageSwitch(PixelShader=@n38.PS, VertexShader=@n37)
   ) {materialFunction=/Game/Weapons/WeaponSkins/Wraps/MaterialFunctions/WorldAlignedTexture}.XYZ Texture) {group=MidasTech_GoldMask, maskChannel=Green, parameterName=GM_TransitionNoiseChannel, defaultValue=<rgba 0 1 0 0>}
  NamedRerouteDeclaration(None=@n39.R) {name=TransitionNoise, nodeColor=<rgba 0 1 0.5078 1>}
  n40 = Power(Base=@n6.R, Exp=ScalarParameter() {defaultValue=1, group=MidasTech_GoldMask, parameterName=GM_MaskPos_Exponent})
  n41 = ScalarParameter() {group=MidasTech_GoldMask, parameterName=gold_radius}
  n42 = SphereMask(
   A=NamedRerouteUsage()
   B=MaterialFunctionCall(
    A=VectorParameter() {group=MidasTech_GoldMask, parameterName=GM_MaskPos_01, defaultValue=<rgba -40 5 100 1>}.RGB
    B=VectorParameter() {group=MidasTech_GoldMask, parameterName=GM_MaskPos_02, defaultValue=<rgba -20 0 160 1>}.RGB
    C=VectorParameter() {group=MidasTech_GoldMask, parameterName=GM_MaskPos_03, defaultValue=<rgba 0 0 180 1>}.RGB
    Alpha=@n40
   ) {materialFunction=&Lerp_3Color}.Result
   Radius=LinearInterpolate(
    A=ScalarParameter() {defaultValue=1, group=MidasTech_GoldMask, parameterName=gold_radius_Min}
    B=ScalarParameter() {defaultValue=55, group=MidasTech_GoldMask, parameterName=gold_radius_Max}
    Alpha=Saturate(None=@n41)
   ) {constA=3.5, constB=52}
   Hardness=ScalarParameter() {defaultValue=0.5, group=MidasTech_GoldMask, parameterName=GM_CoverMaskHardness}
  )
  n43 = Multiply(
   A=NamedRerouteUsage()
   B=LinearInterpolate(
    A=ScalarParameter() {defaultValue=0.5, group=MidasTech_GoldMask, parameterName=GM_Transition_Falloff_Start}
    B=ScalarParameter() {defaultValue=0.1, group=MidasTech_GoldMask, parameterName=GM_Transition_Falloff_End}
    Alpha=@n40
   )
  )
  NamedRerouteDeclaration(None=Saturate(None=Subtract(A=@n42, B=@n43))) {name=TransitionGradient, nodeColor=<rgba 1 0.1875 0 1>}
  NamedRerouteDeclaration(None=@n7.Tech Art Mask) {name=TechArtMask, nodeColor=<rgba 1 0.1875 0 1>}
  ScalarParameter() {parameterName=gold_stage}
 ]
}
