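# Physics asset for the BurgleBoss gadget skeletal mesh, a prop in the CosmeticCompanions emote assets.
# It holds a single body on the root bone, made of one large rotated capsule whose length is tiny next to its radius, so it is nearly a sphere.
# It has no constraints, so it defines one coarse collision volume with no articulated parts.
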PhysicsAsset /CosmeticCompanions/Assets/Props/Emotes/BurgleBoss/Meshes/SKM_BurgleBoss_Gadget_PA
bodies {root [{shapeType=Capsule, length=0.5051, radius=260.5636, rotation=<rot 0 -90 0>}]}
constraints []
